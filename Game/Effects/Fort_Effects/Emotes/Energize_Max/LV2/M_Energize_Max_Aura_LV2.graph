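# Translucent unlit Niagara sprite and mesh material for the level 2 Energize Max emote aura.
# Emissive is particle color times optional panning noise, or scene color via a static switch that zeroes opacity on forward and mobile.
# Opacity combines the aura texture, particle alpha, bottom mask and view-angle fade, with depth fade on default and deferred paths.
# An optional ground-ring parameter distorts aura UVs with radially mapped scrolling smoke noise.
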Material /Game/Effects/Fort_Effects/Emotes/Energize_Max/LV2/M_Energize_Max_Aura_LV2
graph {
 properties {
  bUsedWithNiagaraMeshParticles=true
  bUsedWithNiagaraSprites=true
  blendMode=BLEND_Translucent
  pixelDepthOffsetMode=PDOM_Legacy
  shadingModel=MSM_Unlit
  usedShadingModels=Unlit
 }
 nodes [
  n1 = SceneColor!(UVs)
  n2 = Constant() {r=1}
  n3 = Panner(Coordinate=TextureCoordinate() {uTiling=2, vTiling=0.5}) {speedY=1.5}
  n4 = LinearInterpolate(
   A=Constant() {r=1}
   B=LinearInterpolate(Alpha=TextureSample(UVs=@n3) {samplerType=SAMPLERTYPE_Grayscale, texture=&LowResBlurredNoise}.R) {constA=0.1}
   Alpha=ScalarParameter() {defaultValue=1, parameterName=Use color noise}
  )
  n5 = ParticleColor()
  MaterialFunctionCall(Emissive=StaticSwitchParameter(True=ShadingPathSwitch(Default=@n1.RGB, Deferred=@n1.RGB, Forward=@n2, Mobile=@n2), False=Multiply(A=@n4, B=@n5.RGB)) {parameterName=Use scene color}) {materialFunction=&MF_FX_EmissiveAdjustments}
  n6 = Constant2Vector() {g=0.3, r=10}
  n7 = MaterialFunctionCall() {materialFunction=&VectorToRadialValue}
  n8 = Time()
  n9 = ScalarParameter() {defaultValue=1, parameterName=NoiseSpeed}
  n10 = TextureSample(UVs=Panner(Coordinate=Multiply(A=@n6, B=@n7.Radial Coordinates), Time=Multiply(A=@n8, B=@n9)) {speedX=1, speedY=-2}) {texture=&T_Inky_Smoke_Tile}
  n11 = ScalarParameter() {defaultValue=0.1, parameterName=Noise Internity}
  n12 = Add(
   A=TextureCoordinate()
   B=LinearInterpolate(A=Constant(), B=Multiply(A=@n10.R, B=@n11) {constB=0.1}, Alpha=ScalarParameter() {parameterName=Use ground ring})
  )
  n13 = ComponentMask(None=TextureCoordinate()) {g=true}
  n14 = Multiply(A=Power(Base=OneMinus(None=@n13)) {constExponent=3}) {constB=50}
  n15 = Multiply(
   A=StaticComponentMaskParameter(None=TextureSampleParameter2D(UVs=@n12) {parameterName=Default_Tex, texture=&/T_Energize_Max_Aura_LV2}.RGB) {defaultR=true, parameterName=Component mask}
   B=LinearInterpolate(A=Constant() {r=1}, B=Saturate(None=@n14), Alpha=ScalarParameter() {defaultValue=1, parameterName=Use aura bottom mask})
  )
  n16 = Multiply(
   A=Multiply(A=Constant() {r=1}, B=Multiply(A=@n5.A, B=@n15))
   B=MaterialFunctionCall("FadeSensitivity (optional)"=ScalarParameter() {defaultValue=4, parameterName=Fade edge}) {materialFunction=&FadeBasedOnViewAngle}.Result
  )
  n17 = DepthFade(Opacity=@n16, FadeDistance=ScalarParameter() {defaultValue=50, parameterName=Depth fade}) {fadeDistanceDefault=50}
  n18 = ShadingPathSwitch(Default=@n17, Deferred=@n17, Forward=@n16, Mobile=@n16)
  n19 = Constant()
  StaticSwitchParameter(True=ShadingPathSwitch(Default=@n18, Deferred=@n18, Forward=@n19, Mobile=@n19), False=@n18) {parameterName=Use scene color}
 ]
}
